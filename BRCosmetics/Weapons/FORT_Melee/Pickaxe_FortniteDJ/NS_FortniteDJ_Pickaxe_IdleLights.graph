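# Niagara system for the idle lights on the Fortnite DJ pickaxe, built around a single FakeLight_LWE emitter with no renderers listed.
# System scripts use two Set Variables modules alongside System State.
# Its dynamic inputs include modulo and vector-curve sampling, color construction and rotated position offsets, suggesting a looping, curve-driven light.
# User bools flag front-end and preview contexts, with a front-end float modifier among the inputs.
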NiagaraSystem /BRCosmetics/Weapons/FORT_Melee/Pickaxe_FortniteDJ/NS_FortniteDJ_Pickaxe_IdleLights
dependencies {
 usedDataInterfaces [&NiagaraDataInterfaceVectorCurve]
 usedDynamicInputs [
  &AddVectorToPosition
  /Niagara/DynamicInputs/Vectors/RotateVector
  &Add_Float
  &MakeLinearColorFromVectorAndFloat
  &VectorFromCurve
  &ModuloFloat
  &FN_FrontEndFloatModifier
 ]
 usedModules [
  &SystemState
  &self:SystemSpawnScript.SystemScriptSource.SystemScriptGraph.NiagaraNodeAssignment_1.SetVariables_#1
  &self:SystemSpawnScript.SystemScriptSource.SystemScriptGraph.NiagaraNodeAssignment_2.SetVariables_#2
 ]
}
emitters {FakeLight_LWE {rendererClasses [], renderers []}}
summary {
 systemName=NS_FortniteDJ_Pickaxe_IdleLights
 emitters [{emitterName=FakeLight_LWE, rendererClasses []}]
 userVariables [
  User.bIsFrontEnd {type=NiagaraBool}
  User.bIsFrontEndPreview {type=NiagaraBool}
 ]
}
